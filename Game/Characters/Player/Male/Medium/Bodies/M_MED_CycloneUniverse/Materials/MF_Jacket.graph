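# Material function for the CycloneUniverse player jacket that layers an animated, lava-like glowing surface over input material attributes by Alpha.
# Bump-offset, rotating and panning noise, masked by Gmap, indexes curve-atlas gradients for emissive, base color and roughness.
# Event audio amplitude and spectrum modulate pattern and color, while a HitVector sphere mask brightens hits.
# Switchable flame textures cut the bottom's opacity mask; noise-driven normal displacement adds waves, disabled on mobile.
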MaterialFunction /Game/Characters/Player/Male/Medium/Bodies/M_MED_CycloneUniverse/Materials/MF_Jacket
graph {nodes [
  n1 = FunctionInput() {inputName=Material, inputType=FunctionInput_MaterialAttributes}
  n2 = SphereMask(
   A=VertexInterpolator(VS=PreSkinnedPosition()).PS
   B=VectorParameter() {parameterName=HitVector, defaultValue=<rgba 0 0 0 1>}.RGB
   Radius=ScalarParameter() {parameterName=Radius}
   Hardness=Constant() {r=0.1}
  )
  n3 = BumpOffset(
   Coordinate=Multiply(A=TextureCoordinate(), B=Constant() {r=0.5})
   Height=ScalarParameter() {defaultValue=-0.5, group=Jacket, parameterName=JacketNoiseHeight}
  )
  n4 = Multiply(A=Time(), B=Constant() {r=0.0005})
  n5 = Sine(None=@n4)
  n6 = Frac(None=@n4)
  n7 = FunctionInput() {inputName="Gmap ", inputType=FunctionInput_Scalar}
  n8 = ScalarParameter() {defaultValue=0.05, group=Jacket, parameterName=Jacket Gmap Warp}
  n9 = TextureSample(UVs=Add(
    A=MaterialFunctionCall(UVs=@n3, Rotation Center=@n5, "Rotation Angle (0-1)"=@n6) {materialFunction=&CustomRotator}.Rotated Values
    B=Multiply(A=@n7, B=@n8)
   )) {samplerType=SAMPLERTYPE_Masks, texture=/Game/Global/Textures/Default/Noise/NoisePattern}
  n10 = Multiply(A=@n2, B=@n9.G)
  n11 = ScalarParameter() {defaultValue=0.01, group=Jacket, parameterName=Jacket Pan Speed X}
  n12 = ScalarParameter() {defaultValue=0.01, group=Jacket, parameterName=Jacket Pan Speed Y}
  n13 = TextureSample(UVs=Panner(
    Coordinate=MaterialFunctionCall(UVs=@n3) {materialFunction=&CustomRotator}.Rotated Values
    Time=Time()
    Speed=AppendVector(A=@n11, B=@n12)
   )) {samplerType=SAMPLERTYPE_Masks, texture=&NoisePattern}
  n14 = CollectionParameter() {collection=&FortniteMaterialParameters, parameterName=EventAudio_Amplitude}
  n15 = Power(Base=@n14)
  n16 = LinearInterpolate(A=ScalarParameter() {defaultValue=0.3, group=Jacket, parameterName=BaseFunctionRadius}, B=Constant() {r=0.75}, Alpha=@n15)
  n17 = Saturate(None=Multiply(A=Multiply(A=@n16, B=@n13.B), B=Constant() {r=5}))
  n18 = Constant()
  n19 = TextureCoordinate() {coordinateIndex=1}
  n20 = AppendVector(A=ScalarParameter() {defaultValue=-0.1, group=Jacket, parameterName=X Flame}, B=Constant())
  n21 = AppendVector(A=ScalarParameter() {defaultValue=0.4595, group=Jacket, parameterName=FlameWarp}, B=Constant())
  n22 = TextureSample(UVs=Add(A=Panner(Coordinate=@n19, Speed=@n20), B=Multiply(A=@n13.B, B=@n21))) {samplerType=SAMPLERTYPE_Masks, texture=/Game/Global/Textures/Default/Masks_and_Gradients/FlamePattern}
  n23 = Multiply(A=Time(), B=ScalarParameter() {defaultValue=0.6, group=Jacket, parameterName=Flame Transition Speed})
  n24 = LinearInterpolate(A=@n22.R, B=@n22.R, Alpha=ConstantBiasScale(None=Sine(None=@n23)) {bias=0.5, scale=1})
  n25 = LinearInterpolate(
   A=ScalarParameter() {group=Jacket, parameterName=subtract Min}
   B=ScalarParameter() {defaultValue=-0.35, group=Jacket, parameterName=subtract Max}
   Alpha=@n13.G
  )
  n26 = MaterialFunctionCall(UV Channel=@n19) {materialFunction=&LinearGradient}
  n27 = OneMinus(None=@n26.VGradient)
  n28 = ScalarParameter() {defaultValue=2, group=Jacket, parameterName=EmberTiling}
  n29 = Constant()
  n30 = ScalarParameter() {defaultValue=-1, group=Jacket, parameterName=EmberPanSpeed}
  n31 = ScalarParameter() {defaultValue=0.1095, group=Jacket, parameterName=FW2}
  n32 = Add(
   A=Panner(Coordinate=Multiply(A=@n19, B=@n28) {constB=2}, Speed=AppendVector(A=@n29, B=@n30)) {speedY=-0.75}
   B=Multiply(A=@n13.B, B=AppendVector(A=@n31, B=@n31))
  )
  n33 = Power(
   Base=Multiply(
    A=TextureSample(UVs=@n32) {samplerType=SAMPLERTYPE_Masks, texture=&FlamePattern}.B
    B=ScalarParameter() {defaultValue=2.8, group=Jacket, parameterName=Balance}
   )
   Exp=ScalarParameter() {defaultValue=2, group=Jacket, parameterName=EmberFlamePower}
  ) {constExponent=4}
  n34 = ScalarParameter() {defaultValue=0.2071, group=Jacket, parameterName=StepAdd}
  n35 = Add(
   A=Add(A=Subtract(A=@n24, B=@n25) {constB=0.95}, B=Multiply(A=@n27, B=@n33))
   B=OneMinus(None=MaterialFunctionCall(Alpha=@n26.VGradient, Max=@n34) {materialFunction=&SmoothStep}.Result)
  )
  n36 = MaterialFunctionCall(Alpha=@n27, Max=@n34) {materialFunction=&SmoothStep}
  n37 = StaticSwitchParameter(True=Saturate(None=Multiply(A=@n35, B=@n36.Result)), False=Constant() {r=1}) {parameterName=isBottonJacketFlame}
  n38 = Saturate(None=Multiply(A=OneMinus(None=@n37), B=ScalarParameter() {defaultValue=4, group=Jacket, parameterName=JacketfLame_gradinfuence}))
  n39 = Multiply(A=@n3) {constB=2}
  n40 = Time()
  n41 = ScalarParameter() {defaultValue=0.005, group=Jacket, parameterName=GrimeSimplePanning}
  n42 = LinearInterpolate(
   A=LinearInterpolate(
    A=SphereMask(A=@n9.A, B=@n13.B, Radius=@n17, Hardness=@n18)
    B=ScalarParameter() {defaultValue=0.75, group=Jacket, parameterName=BottomFlameColor}
    Alpha=@n38
   )
   B=TextureSample(UVs=Panner(Coordinate=@n39, Time=@n40, Speed=@n41)) {samplerType=SAMPLERTYPE_Masks, texture=&NoisePattern}.R
   Alpha=Constant() {r=0.33}
  )
  n43 = LinearInterpolate(A=LinearInterpolate(A=Constant(), Alpha=@n42), B=LinearInterpolate(A=Constant() {r=0.8}, Alpha=@n42) {constB=0.99}, Alpha=@n15)
  n44 = Saturate(None=Add(A=Multiply(A=@n10) {constB=1.5}, B=Saturate(None=@n43)))
  n45 = CurveAtlasRowParameter(CurveTime=@n44) {
   atlas=&AthenaCurveAtlas01
   curve=/Game/Characters/Player/Male/Medium/Bodies/M_MED_Cyclone/Textures/Cyclone_Curve01
   defaultValue=213
   parameterName=Gradient
  }
  n46 = CurveAtlasRowParameter(CurveTime=@n44) {
   atlas=&AthenaCurveAtlas01
   curve=&Cyclone_Curve01
   defaultValue=213
   parameterName=Gradient
  }
  n47 = ScalarParameter() {group=Jacket, parameterName=Use Alternate Curve 01}
  n48 = ScalarParameter() {defaultValue=2.5, group=Jacket, parameterName=Lava Brightness}
  n49 = Power(Base=Multiply(A=OneMinus(None=@n7)) {constB=1.5}) {constExponent=4}
  n50 = Multiply(A=Multiply(A=LinearInterpolate(A=@n45, B=@n46, Alpha=@n47), B=Add(A=@n48, B=@n49)), B=@n7)
  n51 = LinearInterpolate(A=@n50, B=Multiply(A=@n50, B=ScalarParameter() {defaultValue=1.1, group=Jacket, parameterName=BottomFlameIntensity}) {constB=4}, Alpha=@n38)
  n52 = ScalarParameter() {defaultValue=4, group=Jacket, parameterName=Hit Intensity}
  n53 = LinearInterpolate(
   A=LinearInterpolate(A=@n51, B=Multiply(A=@n52, B=@n51), Alpha=@n2)
   Alpha=Fresnel(ExponentIn=ScalarParameter() {defaultValue=0.2, group=Jacket, parameterName=Fracket_Fres_EX}) {baseReflectFraction=0}
  ) {constB=0}
  n54 = ScalarParameter() {defaultValue=3, group=Jacket, parameterName=DisplacementFlow_power}
  n55 = Multiply(A=Abs(None=Add(A=@n26.UGradient) {constB=-0.5})) {constB=1.75}
  n56 = Subtract(A=Saturate(None=Power(Base=@n26.VGradient, Exp=@n54)), B=Saturate(None=Power(Base=@n55) {constExponent=2.8}))
  n57 = Multiply(A=TextureCoordinate() {coordinateIndex=1}, B=Constant() {r=0.4})
  n58 = AppendVector(A=Constant(), B=Constant() {r=-0.15})
  n59 = Multiply(
   A=Multiply(A=VertexNormalWS(), B=Saturate(None=@n56))
   B=TextureSample(UVs=Panner(Coordinate=@n57, Speed=@n58)) {samplerType=SAMPLERTYPE_Masks, texture=&NoisePattern}.B
  )
  n60 = ScalarParameter() {group=Jacket, parameterName=VertexWaveIntensity}
  n61 = CollectionParameter() {collection=&FortniteMaterialParameters, parameterName=EventAudio_Spectrum}
  n62 = MaterialFunctionCall(Float4=@n61) {materialFunction=&BreakOutFloat4Components}
  n63 = Multiply(
   A=@n53
   B=LinearInterpolate(
    A=ScalarParameter() {defaultValue=1, group=Jacket, parameterName=Base Color intensity}
    B=ScalarParameter() {defaultValue=1.4, group=Jacket, parameterName=Event Color Kick Bump}
    Alpha=Power(Base=@n62.G)
   )
  )
  n64 = SetMaterialAttributes!(
   MaterialAttributes=@n1
   EmissiveColor=@n53
   Specular=ScalarParameter() {defaultValue=0.1, parameterName=Spec}
   Roughness=LinearInterpolate(A=Constant() {r=0.75}, B=Constant() {r=0.05}, Alpha=@n44) {constA=0.3, constB=0.6}
   ShadingModel=ShadingModel()
   OpacityMask=@n37
   WorldPositionOffset=ShadingPathSwitch(Default=Multiply(A=@n59, B=@n60), Mobile=Constant())
   Normal=GetMaterialAttributes(None=@n1) {attributeGetTypes [#1]}.Normal
   BaseColor=Saturate(None=Multiply(A=@n63, B=@n7))
  ) {attributeSetTypes [
    #2
    #3
    #4
    #5
    #6
    #7
    #1
    #8
   ]}
  FunctionOutput(None=BlendMaterialAttributes(A=@n1, B=@n64, Alpha=FunctionInput() {inputName=Alpha, inputType=FunctionInput_Scalar}))
 ]}
